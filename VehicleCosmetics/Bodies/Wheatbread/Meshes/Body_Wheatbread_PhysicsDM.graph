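# Physics asset for the Wheatbread vehicle body cosmetic, giving it a simplified collision shape.
# It holds a single box body on the root bone, offset forward and upward to enclose the vehicle.
# It has no constraints, so the body simulates as one rigid piece.
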PhysicsAsset /VehicleCosmetics/Bodies/Wheatbread/Meshes/Body_Wheatbread_PhysicsDM
bodies {root [{shapeType=Box, extentX=444, extentY=216, extentZ=116, shapeName=root_box, center=<19.6695 0 103.6695>}]}
constraints []
